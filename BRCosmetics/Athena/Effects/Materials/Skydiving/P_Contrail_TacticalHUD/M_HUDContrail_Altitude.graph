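# Masked Niagara sprite material for the TacticalHUD skydiving contrail that draws a glowing four-digit altitude-style readout.
# Digits come from cloud noise sampled by one world axis (Z by default) plus timed scrolling, not literal height.
# Floor math offsets a number-atlas lookup per quarter of the sprite, masked by particle alpha.
# World height blends emissive from red when low to particle color, overridable by ColorShiftMask and scaled by time of day.
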Material /BRCosmetics/Athena/Effects/Materials/Skydiving/P_Contrail_TacticalHUD/M_HUDContrail_Altitude
graph {
 properties {bUsedWithNiagaraSprites=true, blendMode=BLEND_Masked, pixelDepthOffsetMode=PDOM_Legacy, usedShadingModels=Default Lit}
 nodes [
  n1 = Constant3Vector() {constant=<rgba 10 0.1 0 0>}
  n2 = ParticleColor()
  n3 = Multiply(A=WorldPosition().XYZ) {constB=0.01}
  n4 = Add(B=ComponentMask(None=@n3) {b=true}) {constA=-100, desc=-1 x red height threshold}
  n5 = Clamp(None=Add(
    A=Multiply(A=@n4) {constB=0.002, desc=1 / transition distance from to red}
    B=ScalarParameter() {desc="0 allows color-shift at threshold, 1 prevents it", parameterName=ColorShiftMask}
   ))
  n6 = Constant() {r=0.75}
  n7 = Constant() {r=1}
  n8 = Constant() {r=0.25}
  MaterialFunctionCall(Emissive=Multiply(
    A=LinearInterpolate(A=@n1, B=@n2.RGB, Alpha=@n5)
    B=MaterialFunctionCall(Value_Morning=@n6, Value_Day=@n7, Value_Evening=@n6, Value_Night=@n8) {materialFunction=&Scalar_Time_of_Day}.Result
   )) {materialFunction=&MF_FX_EmissiveAdjustments}
  n9 = StaticComponentMaskParameter(None=Multiply(A=@n3) {constB=0.01}) {defaultB=true, parameterName=AxisTrack}
  n10 = Add(A=AppendVector(A=@n9, B=@n9), B=Multiply(A=ScalarParameter() {parameterName=AddSpeed0-1}, B=Time()) {constA=1})
  n11 = StaticComponentMaskParameter(None=Multiply(A=TextureSample(UVs=@n10) {texture=/Game/Global/Textures/Default/Noise/T_Noise_Cloud}.RGB) {constB=10000}) {defaultB=true, parameterName=AxisTrack}
  n12 = Multiply(A=@n11) {constB=0.02}
  n13 = Multiply(A=Floor(None=Multiply(A=@n12) {constB=0.1})) {constB=0.1}
  n14 = Multiply(A=Constant2Vector() {g=1, r=2}, B=TextureCoordinate() {uTiling=0.2})
  n15 = ComponentMask(None=@n14) {r=true}
  n16 = ComponentMask(None=@n14) {g=true}
  n17 = TextureSample(UVs=AppendVector(A=Add(A=@n13, B=@n15), B=@n16)) {texture=/Game/Environments/Props/Street/Textures/numbers}
  n18 = Add(A=@n15, B=Multiply(A=Floor(None=@n12)) {constB=0.1})
  n19 = TextureSample(UVs=AppendVector(A=Add(A=@n18) {constB=-0.1}, B=@n16)) {texture=&numbers}
  n20 = TextureCoordinate()
  n21 = Ceil(None=ComponentMask(None=Add(A=@n20) {constB=-0.25}) {r=true})
  n22 = Multiply(A=Floor(None=Multiply(A=@n11) {constB=0.1})) {constB=0.1}
  n23 = Multiply(A=Constant2Vector() {g=1, r=2}, B=TextureCoordinate() {uTiling=0.2})
  n24 = Add(A=ComponentMask(None=@n23) {r=true}) {constB=-0.2}
  n25 = ComponentMask(None=@n23) {g=true}
  n26 = TextureSample(UVs=AppendVector(A=Add(A=@n22, B=@n24), B=@n25)) {texture=&numbers}
  n27 = Add(A=@n24, B=Multiply(A=Floor(None=@n11)) {constB=0.1})
  n28 = TextureSample(UVs=AppendVector(A=Add(A=@n27) {constB=-0.1}, B=@n25)) {texture=&numbers}
  n29 = TextureCoordinate()
  n30 = Ceil(None=ComponentMask(None=Add(A=@n29) {constB=-0.75}) {r=true})
  n31 = ComponentMask(None=Add(A=@n20) {constB=-0.5}) {r=true}
  Multiply(
   A=@n2.A
   B=LinearInterpolate(A=LinearInterpolate(A=@n17.R, B=@n19.R, Alpha=@n21), B=LinearInterpolate(A=@n26.R, B=@n28.R, Alpha=@n30), Alpha=Ceil(None=@n31))
  )
 ]
}
