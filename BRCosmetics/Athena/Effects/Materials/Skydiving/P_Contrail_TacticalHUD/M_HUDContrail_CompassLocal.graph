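# Translucent two-sided Niagara sprite material drawing a compass crosshair reticle for the tactical HUD skydiving contrail.
# It samples a sniper crosshair texture twice, in sprite UVs and actor-relative world XY, masking each with thresholded sine bands and blending via Local/World reticule scalars.
# Low quality and mobile replace the texture with thin procedural lines.
# An edge fade along U trims the result, which particle color tints and time of day dims in the alpha calculation.
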Material /BRCosmetics/Athena/Effects/Materials/Skydiving/P_Contrail_TacticalHUD/M_HUDContrail_CompassLocal
graph {
 properties {bUsedWithNiagaraSprites=true, blendMode=BLEND_Translucent, pixelDepthOffsetMode=PDOM_Legacy, twoSided=true, usedShadingModels=Default Lit}
 nodes [
  n1 = Constant() {r=250}
  Divide(A=Constant() {r=1}, B=@n1)
  n2 = MaterialFunctionCall(XY=TextureCoordinate()) {materialFunction=&Swizzle}
  n3 = Add(A=Multiply(A=@n2.YX, B=Constant2Vector() {g=1.5, r=2}) {constB=2}, B=Constant2Vector() {g=0.5, r=0.25}) {constB=0.25}
  n4 = MaterialFunctionCall(Float2=Clamp(None=Sine(None=@n3))) {materialFunction=&BreakOutFloat2Components}
  n5 = Add(A=@n4.R, B=OneMinus(None=@n4.G))
  n6 = Constant() {r=0.5}
  n7 = Multiply(
   A=MaterialFunctionCall(Alpha=Clamp(None=@n5), Min=@n6, Max=@n6) {materialFunction=&SmoothStep}.Result
   B=TextureSample(UVs=@n2.YX) {samplerType=SAMPLERTYPE_LinearColor, texture=/Game/Athena/HUD/reticles/HUD_SniperCrossHairs_v1_Athena}.G
  )
  n8 = Constant() {r=0.997}
  n9 = MaterialFunctionCall(Alpha=@n4.R, Min=@n8, Max=@n8) {materialFunction=&SmoothStep}
  n10 = ShadingPathSwitch(Default=@n7, Deferred=@n7, Forward=@n9.Result, Mobile=@n9.Result)
  n11 = ShadingPathSwitch(Default=@n10, Deferred=@n10, Mobile=@n9.Result) {desc=Migrated from FeatureLevelSwitch}
  n12 = QualitySwitch(Default=@n11, Low=@n9.Result, High=@n11, Medium=@n11)
  n13 = ScalarParameter() {defaultValue=1, parameterName=LocalReticuleOn/Off, sliderMax=1}
  n14 = Add(A=Subtract(A=WorldPosition().XYZ, B=ActorPositionWS()), B=Constant() {r=125})
  n15 = Constant() {r=0.004}
  n16 = MaterialFunctionCall(XY=ComponentMask(None=Multiply(A=@n14, B=@n15) {constB=0.005}) {g=true, r=true}) {materialFunction=&Swizzle}
  n17 = Add(A=Multiply(A=@n16.YX, B=Constant2Vector() {g=1.5, r=2}) {constB=2}, B=Constant2Vector() {g=0.5, r=0.25}) {constB=0.25}
  n18 = MaterialFunctionCall(Float2=Clamp(None=Sine(None=@n17))) {materialFunction=&BreakOutFloat2Components}
  n19 = Add(A=@n18.R, B=OneMinus(None=@n18.G))
  n20 = Constant() {r=0.5}
  n21 = Multiply(
   A=TextureSample(UVs=@n16.YX) {samplerType=SAMPLERTYPE_LinearColor, texture=&HUD_SniperCrossHairs_v1_Athena}.G
   B=MaterialFunctionCall(Alpha=Clamp(None=@n19), Min=@n20, Max=@n20) {materialFunction=&SmoothStep}.Result
  )
  n22 = Constant() {r=0.997}
  n23 = MaterialFunctionCall(Alpha=@n18.R, Min=@n22, Max=@n22) {materialFunction=&SmoothStep}
  n24 = ShadingPathSwitch(Default=ShadingPathSwitch(Default=@n21, Deferred=@n21, Forward=@n23.Result, Mobile=@n23.Result), Mobile=@n23.Result) {desc=Migrated from FeatureLevelSwitch}
  n25 = QualitySwitch(Default=@n24, Low=@n23.Result, High=@n24, Medium=@n24)
  n26 = ScalarParameter() {parameterName=WorldReticuleOn/Off, sliderMax=1}
  n27 = Multiply(A=Add(A=TextureCoordinate()) {constB=-0.5}) {constB=2}
  n28 = MaterialFunctionCall(Float2=Multiply(A=@n27, B=@n27)) {materialFunction=&BreakOutFloat2Components}
  n29 = Add(A=@n28.R, B=@n28.R)
  n30 = Multiply(A=Add(A=OneMinus(None=@n29)) {constB=-0.075}) {constB=20}
  n31 = Multiply(A=Add(A=Multiply(A=@n12, B=@n13), B=Multiply(A=@n25, B=@n26)), B=Saturate(None=Clamp(None=@n30)))
  n32 = ParticleColor()
  MaterialFunctionCall(Emissive=Multiply(A=@n31, B=@n32.RGB)) {materialFunction=&MF_FX_EmissiveAdjustments}
  Multiply(A=@n1) {constB=0.5}
  n33 = Constant() {r=0.75}
  n34 = Constant() {r=1}
  n35 = Constant() {r=0.25}
  Multiply(
   A=@n32.A
   B=Multiply(
    A=@n31
    B=MaterialFunctionCall(Value_Morning=@n33, Value_Day=@n34, Value_Evening=@n33, Value_Night=@n35) {materialFunction=&Scalar_Time_of_Day}.Result
   )
  )
 ]
}
