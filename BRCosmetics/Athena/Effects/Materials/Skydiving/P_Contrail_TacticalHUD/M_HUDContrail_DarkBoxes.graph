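# Translucent Niagara sprite material for the dark boxes of the Tactical HUD skydiving contrail, drawn from a square UI frame texture.
# The frame's alpha tints particle color, is scaled by a time-of-day scalar that dims most at night, then passes through the shared FX emissive adjustment function.
# A separate branch adds, rather than multiplies, frame alpha and particle alpha.
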Material /BRCosmetics/Athena/Effects/Materials/Skydiving/P_Contrail_TacticalHUD/M_HUDContrail_DarkBoxes
graph {
 properties {bUsedWithNiagaraSprites=true, blendMode=BLEND_Translucent, pixelDepthOffsetMode=PDOM_Legacy, usedShadingModels=Default Lit}
 nodes [
  n1 = TextureSample() {texture=/Game/UI/Style/BasicSquareFrame-72x2}
  n2 = ParticleColor()
  Add(A=@n1.A, B=@n2.A)
  n3 = Constant() {r=0.75}
  n4 = Constant() {r=1}
  n5 = Constant() {r=0.25}
  MaterialFunctionCall(Emissive=Multiply(
    A=MaterialFunctionCall(Value_Morning=@n3, Value_Day=@n4, Value_Evening=@n3, Value_Night=@n5) {materialFunction=&Scalar_Time_of_Day}.Result
    B=Multiply(A=@n1.A, B=@n2.RGB)
   )) {materialFunction=&MF_FX_EmissiveAdjustments}
 ]
}
