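# Translucent, two-sided material for Niagara mesh particles, used as a disc in the Tactical HUD skydiving contrail.
# Three sine waves of differing frequency and phase along the V texture coordinate multiply into a procedural band mask, with no textures.
# Two waves are clamped, while the third is remapped to 0–1 and squared.
# The mask scales particle color through MF_FX_EmissiveAdjustments for glow and multiplies particle alpha.
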Material /BRCosmetics/Athena/Effects/Materials/Skydiving/P_Contrail_TacticalHUD/M_HUDContrail_Disc
graph {
 properties {
  bUsedWithNiagaraMeshParticles=true
  blendMode=BLEND_Translucent
  pixelDepthOffsetMode=PDOM_Legacy
  twoSided=true
  usedShadingModels=Default Lit
 }
 nodes [
  n1 = ParticleColor()
  n2 = ComponentMask(None=TextureCoordinate()) {g=true}
  n3 = Add(A=Multiply(A=@n2) {constB=10}) {constB=0.45}
  n4 = Add(A=Multiply(A=@n2) {constB=7}) {constB=0.45}
  n5 = Constant() {r=0.5}
  n6 = Add(A=Multiply(A=Sine(None=@n4), B=@n5) {constB=0.125}, B=@n5) {constB=0.125}
  n7 = Sine(None=Add(A=Multiply(A=@n2) {constB=6}) {constB=0.6})
  n8 = Multiply(A=Clamp(None=Sine(None=@n3)), B=Multiply(A=Multiply(A=@n6, B=@n6), B=Clamp(None=@n7)))
  MaterialFunctionCall(Emissive=Multiply(A=@n1.RGB, B=@n8)) {materialFunction=&MF_FX_EmissiveAdjustments}
  Multiply(A=@n1.A, B=@n8)
 ]
}
