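# Translucent two-sided material for the pointer of the tactical HUD skydiving contrail effect, set up for Niagara mesh particles.
# An inverted low-exponent Fresnel fades silhouette edges, multiplying both particle color and particle alpha.
# The color product feeds an emissive adjustment function.
# The alpha product is scaled by a time-of-day scalar that is full in daytime, slightly reduced at morning and evening, and much dimmer at night.
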Material /BRCosmetics/Athena/Effects/Materials/Skydiving/P_Contrail_TacticalHUD/M_HUDContrail_Pointer
graph {
 properties {
  bUsedWithNiagaraMeshParticles=true
  blendMode=BLEND_Translucent
  pixelDepthOffsetMode=PDOM_Legacy
  twoSided=true
  usedShadingModels=Default Lit
 }
 nodes [
  n1 = ParticleColor()
  n2 = OneMinus(None=Fresnel() {exponent=0.25})
  MaterialFunctionCall(Emissive=Multiply(A=@n1.RGB, B=@n2)) {materialFunction=&MF_FX_EmissiveAdjustments}
  n3 = Constant() {r=0.75}
  Multiply(
   A=Multiply(A=@n1.A, B=@n2)
   B=MaterialFunctionCall(Value_Morning=@n3, Value_Day=Constant() {r=1}, Value_Evening=@n3, Value_Night=Constant() {r=0.125}) {materialFunction=&Scalar_Time_of_Day}.Result
  )
 ]
}
