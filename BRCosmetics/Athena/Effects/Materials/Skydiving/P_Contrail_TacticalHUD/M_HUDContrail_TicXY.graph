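# Translucent Niagara sprite material for the tick marks of the Tactical HUD skydiving contrail cosmetic.
# A HUD crosshair line texture's red channel masks particle color, which feeds the shared FX emissive adjustments function.
# The same mask multiplied by particle alpha feeds a SmoothStep with equal min and max, acting as a hard cutoff rather than a soft fade.
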Material /BRCosmetics/Athena/Effects/Materials/Skydiving/P_Contrail_TacticalHUD/M_HUDContrail_TicXY
graph {
 properties {bUsedWithNiagaraSprites=true, blendMode=BLEND_Translucent, pixelDepthOffsetMode=PDOM_Legacy, usedShadingModels=Default Lit}
 nodes [
  n1 = ParticleColor()
  n2 = TextureSample() {texture=/Game/UI/HUD/Reticles/HUD_Xhair_2x10_line}
  MaterialFunctionCall(Emissive=Multiply(A=@n1.RGB, B=@n2.R)) {materialFunction=&MF_FX_EmissiveAdjustments}
  n3 = Constant() {r=0.5}
  MaterialFunctionCall(Alpha=Multiply(A=@n1.A, B=@n2.R), Min=@n3, Max=@n3) {materialFunction=&SmoothStep}
 ]
}
