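# Translucent unlit Niagara mesh-particle material drawing scrolling altimeter-style tick marks for the Tactical HUD skydiving contrail.
# Vertex-color red indexes each tick; world position offset billboards it and snaps it to a world-Z grid from particle position plus optional time scroll.
# Every fourth grid tick gets a larger scale and different sine-based line mask.
# Vertex-color green collapses vertices along normals and zeroes opacity, which also uses particle alpha and camera-depth fade.
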Material /BRCosmetics/Athena/Effects/Materials/Skydiving/P_Contrail_TacticalHUD/M_HUD_Contrail_ScrollingTics
graph {
 properties {
  bUsedWithNiagaraMeshParticles=true
  blendMode=BLEND_Translucent
  pixelDepthOffsetMode=PDOM_Legacy
  shadingModel=MSM_Unlit
  usedShadingModels=Unlit
 }
 nodes [
  n1 = ScalarParameter() {defaultValue=16, parameterName=NumberOfTicks}
  n2 = Multiply(A=VertexColor().R, B=@n1) {constB=16}
  n3 = Multiply(A=ScalarParameter() {parameterName=AddSpeed0-1}) {constB=-1000}
  n4 = AppendVector(A=Constant2Vector(), B=Time())
  n5 = Multiply(A=Add(A=ParticlePositionWS(), B=Multiply(A=@n3, B=@n4) {constA=1})) {constB=0.15}
  n6 = ScalarParameter() {defaultValue=15, parameterName=DistanceBetweenTicks}
  n7 = Floor(None=Divide(A=ComponentMask(None=@n5) {b=true}, B=@n6))
  n8 = Abs(None=Ceil(None=Add(A=@n2, B=@n7)))
  n9 = Constant() {r=4}
  n10 = OneMinus(None=Saturate(None=Fmod(A=@n8, B=@n9)))
  n11 = MaterialFunctionCall!(
   Center Location
   0-1 UVs
   XY scale=LinearInterpolate(A=Constant2Vector() {g=2, r=210}, B=Constant2Vector() {g=4, r=240}, Alpha=@n10) {constAlpha=0}
   OverridePivot
   "Normals (Optional)"
  ) {materialFunction=&Sprite}
  n12 = Floor(None=Divide(A=ComponentMask(None=@n5) {b=true}, B=@n6))
  n13 = Constant3Vector() {constant=<rgba 0 0 1 0>}
  n14 = Multiply(A=Subtract(A=Multiply(A=@n12, B=@n6) {constB=100}, B=ComponentMask(None=@n5) {b=true}), B=@n13)
  n15 = Multiply(A=VertexColor().R, B=@n1) {constB=16}
  n16 = Multiply(A=@n13, B=Divide(A=Ceil(None=@n15), B=@n1) {constB=16})
  n17 = Multiply(A=@n1, B=@n6)
  n18 = VertexColor()
  LinearInterpolate(
   A=Add(A=Add(A=@n11.World position offset, B=@n14), B=Multiply(A=@n16, B=@n17))
   B=Multiply(A=VertexNormalWS(), B=ScalarParameter() {defaultValue=-1000}) {constB=-250}
   Alpha=@n18.G
  )
  n19 = ParticleColor()
  MaterialFunctionCall(Emissive=@n19.RGB) {materialFunction=&MF_FX_EmissiveAdjustments}
  n20 = TextureCoordinate() {uTiling=4}
  n21 = ComponentMask(None=@n20) {r=true}
  n22 = Sine(None=Add(A=Multiply(A=@n21) {constB=0.175}) {constB=0.395})
  n23 = ComponentMask(None=@n20) {r=true}
  n24 = Sine(None=Add(A=Multiply(A=@n23) {constB=0.2}) {constB=0.345})
  n25 = LinearInterpolate(A=Clamp(None=Multiply(A=@n22) {constB=50}), B=Clamp(None=Multiply(A=@n24) {constB=20}), Alpha=@n10)
  n26 = ScalarParameter() {parameterName=CameraMask}
  n27 = Constant() {r=100}
  n28 = MaterialFunctionCall(Fade Length=@n27, Fade Offset=@n27) {materialFunction=&CameraDepthFade}
  Multiply(A=Multiply(A=@n19.A, B=LinearInterpolate(A=@n25, Alpha=@n18.G) {constA=1, constB=0}), B=Clamp(None=Add(A=@n26, B=@n28.Result)))
 ]
}
